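# Master translucent unlit, two-sided wind material for the Emperor glider FX, usable on Niagara sprites, ribbons and meshes.
# Color comes from spiral-blurred scene texture plus an air tint, replaced by black plus tint on mobile.
# A panning, noise-distorted airflow mask multiplies scene-color output, V-gradient fades, view-angle falloff, depth fade and particle alpha for opacity.
# Switches select radial UVs and Niagara-driven time.
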Material /BRCosmetics/Gadgets/Assets/VinderTech_GliderChute/Glider_Emperor/FX/Material/M_Master_Wind
graph {
 properties {
  bUsedWithNiagaraMeshParticles=true
  bUsedWithNiagaraRibbons=true
  bUsedWithNiagaraSprites=true
  blendMode=BLEND_Translucent
  pixelDepthOffsetMode=PDOM_Legacy
  shadingModel=MSM_Unlit
  translucencyPass=MTP_BeforeDOF
  twoSided=true
  usedShadingModels=Unlit
 }
 nodes [
  n1 = VectorParameter() {parameterName=FinalColorScale, defaultValue=<rgba 1 1 1 1>}
  n2 = MaterialFunctionCall!(
   Distance=ScalarParameter() {defaultValue=0.02, parameterName=SBlur_Distance}
   Distance Steps=ScalarParameter() {defaultValue=8, parameterName=SBlur_DistanceSteps}
   Radial Steps=ScalarParameter() {defaultValue=8, parameterName=SBlur_RadialSteps}
   Radial Offset
   TempAA Radial Blur
   TempAA Distance Blur
   Distance Mask
  ) {materialFunction=/Engine/Functions/Engine_MaterialFunctions01/Texturing/SpiralBlur-SceneTexture}
  n3 = Constant3Vector()
  n4 = Add(
   A=ShadingPathSwitch(Default=@n2.Result, Deferred=@n2.Result, Mobile=@n3) {desc=Migrated from FeatureLevelSwitch}
   B=VectorParameter() {parameterName=AirColor, defaultValue=<rgba 0 0 0 1>}.RGB
  )
  n5 = ParticleColor()
  MaterialFunctionCall(Emissive=Multiply(A=Multiply(A=@n4, B=@n1.RGB), B=@n5.RGB)) {materialFunction=&MF_FX_EmissiveAdjustments}
  n6 = VectorParameter() {parameterName=UVPanner}
  n7 = StaticSwitchParameter(
   True=MaterialFunctionCall() {materialFunction=&VectorToRadialValue}.Radial Coordinates
   False=TextureCoordinate()
  ) {parameterName=RadialUV}
  n8 = VectorParameter() {parameterName="UVScale&&Offset", defaultValue=<rgba 1 1 0 0>}
  n9 = Add(
   A=Multiply(
    A=MaterialFunctionCall(UVs=@n7, "Rotation Angle (0-1)"=@n6.B) {materialFunction=&CustomRotator}.Rotated Values
    B=AppendVector(A=@n8.R, B=@n8.R)
   )
   B=AppendVector(A=@n8.B, B=@n8.B)
  )
  n10 = StaticSwitchParameter(True=DynamicParameter() {paramNames [TimeInput,Param2,Param3,Param4]}.TimeInput, False=Time() {bOverride_Period=true, period=360}) {parameterName=UseNiagaraTime}
  n11 = Multiply(A=AppendVector(A=@n6.R, B=@n6.R), B=@n10)
  n12 = VectorParameter() {parameterName=Noise1_UVPanner}
  n13 = VectorParameter() {parameterName="Noise1_UVScale&&Offset", defaultValue=<rgba 1 1 0 0>}
  n14 = Multiply(
   A=MaterialFunctionCall(UVs=@n7, "Rotation Angle (0-1)"=@n12.B) {materialFunction=&CustomRotator}.Rotated Values
   B=AppendVector(A=@n13.R, B=@n13.R)
  )
  n15 = AppendVector(A=@n13.B, B=@n13.B)
  n16 = AppendVector(A=@n12.R, B=@n12.R)
  n17 = TextureSampleParameter2D(UVs=Add(A=Add(A=@n14, B=@n15), B=Multiply(A=@n16, B=@n10))) {parameterName=Noise1_Texture}
  n18 = AppendVector(A=@n17.R, B=@n17.R)
  n19 = VectorParameter() {parameterName=Noise1_Str}
  n20 = AppendVector(A=@n19.R, B=@n19.R)
  n21 = TextureSampleParameter2D(UVs=Add(A=Add(A=@n9, B=@n11), B=Multiply(A=@n18, B=@n20))) {parameterName=AirFlow1_Texture}
  n22 = ScalarParameter() {parameterName=AirFlow1_Scale}
  n23 = Multiply(
   A=ShadingPathSwitch(Default=@n2.SceneColor clamp to 0, Deferred=@n2.SceneColor clamp to 0, Mobile=@n3) {desc=Migrated from FeatureLevelSwitch}
   B=Saturate(None=Multiply(A=@n21.R, B=@n22))
  )
  n24 = MaterialFunctionCall(Float2=@n7) {materialFunction=&BreakOutFloat2Components}
  n25 = OneMinus(None=Power(Base=@n24.G, Exp=ScalarParameter() {parameterName=AlphaPower}))
  n26 = Saturate(None=Add(A=@n24.G, B=ScalarParameter() {parameterName=AlphaPower2Offset}))
  n27 = ScalarParameter() {parameterName=AlphaPower2}
  n28 = DotProduct(A=CameraVectorWS(), B=PixelNormalWS())
  n29 = Multiply(
   A=Multiply(A=Multiply(A=@n23, B=@n25), B=Power(Base=@n26, Exp=@n27))
   B=Power(Base=Abs(None=@n28), Exp=ScalarParameter() {parameterName=FresnelExponentIn})
  )
  n30 = ScalarParameter() {parameterName=DistanceFade}
  Multiply(A=Multiply(A=DepthFade(Opacity=@n29, FadeDistance=@n30), B=ScalarParameter() {parameterName=FinalAlphaScale}), B=@n5.A)
 ]
}
