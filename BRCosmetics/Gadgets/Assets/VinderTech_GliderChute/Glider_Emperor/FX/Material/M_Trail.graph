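# Unlit, two-sided translucent trail material for the Emperor glider's FX, flagged for Niagara ribbons.
# A scaled, offset and time-panned silk texture passes through MF_FX_Refine, with a dynamic "Power" parameter driving part of the refinement.
# The result, multiplied by a gradient mask and particle alpha, feeds a depth fade for soft intersections.
# Emissive comes from a color parameter times particle color through MF_FX_EmissiveAdjustments.
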Material /BRCosmetics/Gadgets/Assets/VinderTech_GliderChute/Glider_Emperor/FX/Material/M_Trail
graph {
 properties {
  bUsedWithNiagaraRibbons=true
  blendMode=BLEND_Translucent
  pixelDepthOffsetMode=PDOM_Legacy
  shadingModel=MSM_Unlit
  twoSided=true
  usedShadingModels=Unlit
 }
 nodes [
  n1 = VectorParameter() {parameterName=Texture01UVScaleAndOffset, defaultValue=<rgba 1 1 0 0>}
  n2 = Add(
   A=MaterialFunctionCall(UVs=TextureCoordinate(), Texture Scale=AppendVector(A=@n1.RGB, B=@n1.RGB)) {materialFunction=&ScaleUVsByCenter}.UVs
   B=AppendVector(A=@n1.G, B=@n1.G)
  )
  n3 = VectorParameter() {parameterName=Texture01UVPanner}
  n4 = Multiply(A=Time(), B=AppendVector(A=@n3.R, B=@n3.R))
  n5 = VectorParameter() {group=Emissive, parameterName=Texture01Refine, sortPriority=4, defaultValue=<rgba 1 1 1 0.5>}
  n6 = DynamicParameter() {paramNames [Power,Param2,Param3,Param4]}
  n7 = MaterialFunctionCall(
   Texture=TextureSampleParameter2D(UVs=Add(A=@n2, B=@n4)) {parameterName=Texture01, texture=/Game/Gadgets/Assets/VinderTech_GliderChute/Glider_Emperor/FX/Texture/T_Emperor_Silk_01}.R
   RGB=MaterialFunctionCall(X=Multiply(A=@n5.R, B=@n6.Power), Y=@n5.G, Z=@n5.G) {materialFunction=&MakeFloat3}.Result
   A=@n5.A
  ) {materialFunction=&MF_FX_Refine}
  n8 = TextureSampleParameter2D() {parameterName=OpacityMask, texture=/BRCosmetics/Gadgets/Assets/VinderTech_GliderChute/Glider_Emperor/FX/Texture/T_Emperor_Grad_01}
  n9 = ParticleColor()
  DepthFade(Opacity=Multiply(A=Multiply(A=@n7.Result, B=@n8.R), B=@n9.A), FadeDistance=ScalarParameter() {parameterName=DIstanceFade})
  MaterialFunctionCall(Emissive=Multiply(A=VectorParameter() {parameterName=Color, defaultValue=<rgba 0 0 0 1>}.RGB, B=@n9.RGB)) {materialFunction=&MF_FX_EmissiveAdjustments}
 ]
}
