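# Niagara system for the Emperor glider, layering deploy burst flares, air-break and airflow meshes, ribbons and anim trails.
# Emitters follow bones and sockets on user-supplied skeletal meshes, mostly starting after InitDelay and fading in by a shared curve.
# System Interp and Integrator values blend opacity and ribbon velocity, while normalized camera distance adjusts size, lifetime and alpha.
# An isDark parameter switches red or warm colors to dark or white variants; front-end flags suppress the burst.
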NiagaraSystem /BRCosmetics/Gadgets/Assets/VinderTech_GliderChute/Glider_Emperor/FX/NS_Glider_Emperor
dependencies {
 usedDataInterfaces [&NiagaraDataInterfaceArrayFloat, &NiagaraDataInterfaceCamera]
 usedDynamicInputs []
 usedModules [
  &SystemState
  &self:Interp
  &self:CameraDistance
 ]
}
emitters {
 AirBreak {
  emitterUpdateScript [
   &EmitterState(Delay First Loop Only=true, Life Cycle Mode=::NewEnumerator1, Loop Delay=$User.InitDelay, UseLoopDelay=true)
   &SpawnRate(SpawnRate=300, Use Spawn Probability=false)
  ]
  particleSpawnScript [
   &InitializeParticle(
    Color Mode=::NewEnumerator0
    Lifetime=0.03
    Mesh Scale Max=<0.35 0.51 0.51>
    Mesh Scale Min=<0.29 0.41 0.41>
    Mesh Scale Mode=::NewEnumerator2
    UsePositionOffset=false
    Write Material Random=true
   )
   &InitialMeshOrientation(
    Rotation=&RandomRangeVector(Minimum=<-1 0 0>, Maximum=<1 0 0>)
    Use Orientation Vector=true
    Use Rotation Vector=true
   )
   &FN_SetSpawnGroup()
  ]
  particleUpdateScript [
   &ParticleState()
   &ScaleMeshSize(Scale Factor=&VectorFromFloat(Value=&FloatFromCurve(FloatCurve=DataInterface() {propertyValues {Curve {keys [<at 0 0.5>,<at 0.3 0.5>,<at 1 7>]}}}, CurveIndex=$System.CameraDistanceNormalized))) {module=ScaleMeshSize001}
   &ScaleMeshSize(Scale Factor=&VectorFromFloat(Value=&FloatFromCurve(
      FloatCurve=DataInterface() {propertyValues {Curve {keys [
          {arriveTangent=23.7565, interpMode=RCIM_Cubic, leaveTangent=23.7565, tangentMode=RCTM_User, time=0, value=0}
          {interpMode=RCIM_Cubic, tangentMode=RCTM_User, time=0.1, value=1}
          {arriveTangent=-3.434, interpMode=RCIM_Cubic, tangentMode=RCTM_Break, time=1, value=0}
         ]}}}
      Scale Curve=&SelectFloatFromArray(Direct Array Index=$Particles.SpawnGroup, Float Selection Array=DataInterface() {propertyValues {FloatData [1,1,0,0]}})
     )))
   &SkeletalMeshLocation(
    Bone / Socket Index=$Particles.SpawnGroup
    Bone Sampling Mode=::NewEnumerator4
    MaskSpawnGroup=false
    Orientation Sampling=::NewEnumerator1
    Skeletal Mesh=$User.SkeletalMesh002
   )
   &CameraOffset(Camera Offset Amount=&FloatFromCurve(FloatCurve=DataInterface() {propertyValues {Curve {keys [<at 0 10>,<at 0.3 10>,<at 1 30>]}}}, CurveIndex=$System.CameraDistanceNormalized))
   &ScaleColor(
    Scale Alpha=&Add_Float(
     A=&Lerp_Float(A=0.15, B=0.3, Alpha=$System.FN_Integrator.Value)
     B=&Lerp_Float(
      A=0
      B=0.7
      Alpha=&FloatFromCurve(
       FloatCurve=DataInterface() {propertyValues {Curve {keys [<at 0 0>,<at 0.7 0.362>,<at 1 1>]}}}
       CurveIndex=&ClampFloat(Float=$System.Interp.Value)
      )
     )
    )
    ScaleA=true
    ScaleRGB=true
   ) {module=ScaleColor001}
   &ScaleColor(
    Scale Alpha=&FloatFromCurve(
     FloatCurve=$User.FadeinCurve
     CurveIndex=&ClampFloat(Float=&Subtract_Float(A=$System.Age, B=$User.InitDelay))
     Scale Curve=1.5
    )
    ScaleA=true
    ScaleRGB=true
   ) {module=ScaleColor003}
   &ScaleColor(
    Scale Alpha=&FloatFromCurve(FloatCurve=DataInterface() {propertyValues {Curve {keys [<at 0 1>,<at 0.3 1>,<at 1 0.13>]}}}, CurveIndex=$System.CameraDistanceNormalized)
    Scale RGB=&VectorFromFloat(Value=&FloatFromCurve(FloatCurve=DataInterface() {propertyValues {Curve {keys [<at 0 1>,<at 0.3 1>,<at 1 1.5>]}}}, CurveIndex=$System.CameraDistanceNormalized))
    ScaleA=true
    ScaleRGB=true
   ) {module=ScaleColor006}
  ]
  rendererClasses [&NiagaraMeshRendererProperties]
  renderers [{rendererClass=&NiagaraMeshRendererProperties, rendererIndex=0}]
 }
 Convergent_airflow_start {
  emitterUpdateScript [
   &EmitterState(
    Inactive Response=::NewEnumerator1
    Life Cycle Mode=::NewEnumerator1
    Loop Behavior=::NewEnumerator1
    Loop Delay=$User.InitDelay
    Loop Duration Mode=::NewEnumerator1
    UseLoopDelay=true
   )
   &SpawnRate(SpawnRate=25, Use Spawn Probability=false)
  ]
  particleSpawnScript [
   /Niagara/Modules/Spawn/Initialization/InitializeParticle(
    Color=<rgba 0.5352 0.6839 0.7917 0.8>
    Lifetime=&RandomRangeFloat(Minimum=0.5, Maximum=0.5)
    Mesh Scale=&Multiply_Vector(A=&RandomRangeVector(Minimum=<0.7 0.3 0.7>, Maximum=<0.95 0.95 0.95>), B=<0.7 1.2 0.7>)
    Module.Write Color=true
    Module.Write Lifetime=true
    Module.Write Mass=true
    Module.Write Material Random=true
    Module.Write Position=false
    Module.Write Scale=true
    Module.Write SpriteRotation=false
    Module.Write SpriteSize=true
    Position=&SimulationPosition()
    Sprite Size=<5 5>
   )
   /SetVariables(Particles.MeshOrientation=&MultiplyQuaternion(
     Quaternion A=&MultiplyQuaternion(
      Quaternion A=&MakeQuatFromAxisAngle(Axis=<1 0 0>, AngleInDegrees=180)
      Quaternion B=&MakeQuatFromAxisAngle(Axis=<0 0 1>, AngleInDegrees=&RandomRangeFloat(Minimum=0, Maximum=360))
     )
     Quaternion B=&MakeQuatFromAxisAngle(Axis=<1 0 0>, AngleInDegrees=65.8161)
    )) {isSetParameters=true}
  ]
  particleUpdateScript [
   &ParticleState()
   &SolveForcesAndVelocity(Clamp Velocity=false, Limit Acceleration=false)
   &DynamicMaterialParameters(
    Index 0 Param 1=&FloatFromCurve(FloatCurve=DataInterface() {propertyValues {Curve {keys [{tangentMode=RCTM_User, time=0, value=1}, {tangentMode=RCTM_User, time=1, value=0}]}}})
    Index 0 Param 2=1
    Index 0 Param 3=&Multiply_Float(
     A=&RandomRangeFloat(Minimum=2, Maximum=2)
     B=&FloatFromCurve(FloatCurve=DataInterface() {propertyValues {Curve {keys [<at 0 1>,<at 1 0.5>]}}}, Scale Curve=1.5)
    )
    Module.Param0WriteEnabled=true
    Module.Param1WriteEnabled=true
    Module.Param2WriteEnabled=true
    Module.Param3WriteEnabled=true
   )
   &ScaleColor(
    Scale Alpha=&Multiply_Float(
     A=&Multiply_Float(
      A=&Multiply_Float(A=1, B=&FNTimeOfDayScalar(Night Scale=1, Morning Scale=0.3, Evening Scale=0.4))
      B=&FloatFromCurve(FloatCurve=DataInterface() {propertyValues {Curve {keys [
           <at 0 0 RCIM_Cubic>
           <at 0.1912 0.9932>
           {arriveTangent=-1.2279, interpMode=RCIM_Cubic, leaveTangent=-1.2279, time=0.7879, value=1.0115}
           <at 1 0>
          ]}}})
     )
     B=&FloatFromCurve(FloatCurve=DataInterface() {propertyValues {Curve {keys [<at 0 0>,<at 0.2 0.5258>]}}}, CurveIndex=$Emitter.Age, Scale Curve=0.2)
    )
    ScaleA=true
    ScaleRGB=true
   )
   &ScaleMeshSize(Scale Factor=&Multiply_Vector(
     A=<1 1 0.7>
     B=&VectorFromFloat(Value=&FloatFromCurve(FloatCurve=DataInterface() {propertyValues {Curve {keys [<at 0 1>,<at 1 1>]}}}, CurveIndex=$Emitter.Age))
    ))
   &CameraOffset(Camera Offset Amount=&FloatFromCurve(FloatCurve=DataInterface() {propertyValues {Curve {keys [<at 0 0>,<at 1 50>]}}})) {enabled=false}
   &ScaleColor(
    Scale Alpha=&Add_Float(
     A=&Lerp_Float(A=1, B=0.55, Alpha=$System.FN_Integrator.Value)
     B=&Lerp_Float(
      A=0
      B=-0.55
      Alpha=&FloatFromCurve(
       FloatCurve=DataInterface() {propertyValues {Curve {keys [<at 0 0>,<at 0.8 0.3>,<at 1 1>]}}}
       CurveIndex=&ClampFloat(Float=$System.Interp.Value)
      )
     )
    )
    ScaleA=true
    ScaleRGB=true
   ) {module=ScaleColor001}
   &SkeletalMeshLocation(
    Bone / Socket Index=4
    Bone Sampling Mode=::NewEnumerator4
    MaskSpawnGroup=false
    Orientation Sampling=::NewEnumerator1
    Sampled Position Offset=<-95 0 0>
    Skeletal Mesh=$User.SkeletalMesh
   )
   &SkeletalMeshLocation(
    Bone Sampling Mode=::NewEnumerator6
    MaskSpawnGroup=false
    Orientation Sampling=::NewEnumerator1
    Position Sampling=::NewEnumerator2
    Skeletal Mesh=$User.SkeletalMesh
   ) {module=SkeletalMeshLocation001}
   /SetVariables(Particles.NewO:Quat4f=&MultiplyQuaternion(
     Quaternion A=&MultiplyQuaternion(
      Quaternion A=&FindQuatBetweenTwoVectors(
       From Vector=<0 0 -1>
       To Vector=&Multiply_Vector(
        A=&NormalizeVector(
         Vector To Normalize=&Subtract_Vector(
          A=Linked() {linkedVariable {name=Particles.SkeletalMeshLocation001.SampledPosition, type=NiagaraPosition}}
          B=$Particles.SkeletalMeshLocation.SampledPosition
         )
         Safe Normalize=true
         Safe Fallback Threshold=0
         Safe Fallback Vector=<0 0 0.001>
        )
        B=<-1 1 1>
       )
      )
      Quaternion B=$Particles.MeshOrientation
     )
     Quaternion B=<0 0 0 1>
    )) {isSetParameters=true}
   &ScaleColor(
    Scale Alpha=&FloatFromCurve(
     FloatCurve=$User.FadeinCurve
     CurveIndex=&ClampFloat(Float=&Subtract_Float(A=$System.Age, B=$User.InitDelay))
    )
    ScaleA=true
    ScaleRGB=true
   ) {module=ScaleColor003}
  ]
  rendererClasses [&NiagaraMeshRendererProperties]
  renderers [{rendererClass=&NiagaraMeshRendererProperties, rendererIndex=0}]
 }
 Empty003 {
  emitterUpdateScript [
   &EmitterState(
    Delay First Loop Only=true
    Inactive Response=::NewEnumerator1
    Life Cycle Mode=::NewEnumerator1
    Loop Delay=$User.InitDelay
    UseLoopDelay=true
   )
   &SpawnRate(SpawnRate=240, Use Spawn Probability=false) {enabled=false}
   &FN_SpawnRibbon_v2(Number of Ribbons=2, SpawnRate=50)
  ]
  particleSpawnScript [
   /Game/Effects/Niagara/Modules/Ribbons/FN_InitializeRibbon_v4(
    Color:LinearColor=<rgba 1 0 0 2>
    Lifetime:NiagaraFloat=&FloatFromCurve(
     FloatCurve=DataInterface() {propertyValues {Curve {keys [<at 0 0.4>,<at 0.3 0.4>,<at 1 0.65>]}}}
     CurveIndex=$System.CameraDistanceNormalized
    )
    Offset Axis:ENiagaraOrientationAxis=::ZAxis
    Ribbon Width=40
   )
   &SkeletalMeshLocation(
    Bone / Socket Index=$Particles.SpawnGroup
    MaskSpawnGroup=false
    Mesh Sampling Type=::NewEnumerator5
    Skeletal Mesh=$User.SkeletalMesh_CapeRed
    Socket Sampling Mode=::NewEnumerator5
   )
   /SetVariables(Particles.RibbonWidth=&SelectFloatFromArray(Direct Array Index=$Particles.SpawnGroup, Float Selection Array=DataInterface() {propertyValues {FloatData [30,30,60,60]}})) {isSetParameters=true}
   &AddVelocity(Velocity=<0 -1000 0>) {enabled=false}
  ]
  particleUpdateScript [
   /SetVariables(Particles.Velocity=&TransformVector(
     Vector=&Lerp_Vector(
      A=<0 0 1000>
      B=<0 -700 0>
      Alpha=&FloatFromCurve(
       FloatCurve=DataInterface() {propertyValues {Curve {keys [
           {arriveTangent=4.1785, interpMode=RCIM_Cubic, leaveTangent=4.1785, tangentMode=RCTM_User, time=0, value=0}
           {arriveTangent=0.9995, interpMode=RCIM_Cubic, leaveTangent=0.9995, tangentMode=RCTM_User, time=0.2223, value=0.6825}
           {arriveTangent=-0.0405, interpMode=RCIM_Cubic, leaveTangent=-0.0405, tangentMode=RCTM_User, time=1, value=1}
          ]}}}
       CurveIndex=$System.FN_Integrator.Value
      )
     )
     Destination Space=::World
    )) {isSetParameters=true}
   &SolveForcesAndVelocity(Clamp Velocity=false, Limit Acceleration=false)
   /Game/Effects/Niagara/Modules/Ribbons/FN_UpdateRibbon_v4(Detach Locked Particle=true, Locked Particle Detach Time=0.001)
   /Game/Effects/Niagara/Modules/Materials/FN_RibbonMaterial/FN_RibbonMaterial_v2(
    Color:LinearColor=&MakeLinearColor(R:NiagaraFloat=1, G:NiagaraFloat=1, B:NiagaraFloat=1, A:NiagaraFloat=1)
    Color Pan U Offset:NiagaraFloat=&SelectFloatFromArray(Direct Array Index=$Particles.SpawnGroup, Float Selection Array=DataInterface() {propertyValues {FloatData [0,12.3925,158.6253,64]}})
    Color Pan U Speed=2.5
    Opacity Pan U Offset=0.5
    Opacity Pan U Speed=2
    Write to Module Namespace=true
   )
   &ScaleColor(
    Scale Alpha=&FloatFromCurve(
     FloatCurve=DataInterface() {propertyValues {Curve {keys [{interpMode=RCIM_Cubic, leaveTangent=4.3733, tangentMode=RCTM_Break, time=0, value=0}, <at 0.5 1>]}}}
     CurveIndex=$Engine.System.Age
    )
    ScaleA=true
    ScaleRGB=true
   )
   &ScaleColor(
    Scale Alpha=&FloatFromCurve(FloatCurve=DataInterface() {propertyValues {Curve {keys [
         {arriveTangent=38.4055, interpMode=RCIM_Cubic, leaveTangent=38.4055, tangentMode=RCTM_User, time=0, value=0}
         {interpMode=RCIM_Cubic, tangentMode=RCTM_Break, time=0.0689, value=1}
         {arriveTangent=-3.084, interpMode=RCIM_Cubic, leaveTangent=-3.084, tangentMode=RCTM_User, time=0.7, value=0.3}
        ]}}})
    ScaleA=true
    ScaleRGB=true
   ) {module=ScaleColor001}
   &ScaleColor(
    Scale Alpha=&Add_Float(
     A=&Lerp_Float(A=0.4, B=0.7, Alpha=$System.FN_Integrator.Value)
     B=&Lerp_Float(
      A=0
      B=0.4
      Alpha=&FloatFromCurve(
       FloatCurve=DataInterface() {propertyValues {Curve {keys [<at -0.1 -4>,<at 0 0>,<at 0.8 0.6>,<at 1 1>]}}}
       CurveIndex=&ClampFloat(Float=$System.Interp.Value, Min=-1)
      )
     )
    )
    ScaleA=true
    ScaleRGB=true
   ) {module=ScaleColor002}
   &ScaleColor(
    Scale Mode=::NewEnumerator1
    Scale RGBA=/Niagara/DynamicInputs/TypeConversions/MakeCustomVector4FromBool(
     Bool:NiagaraBool=&SetBoolByFloatComparison(A=$User.isDark, B=0)
     True Vector4:Vector4f=<0 0 0 2.5>
     False Vector4:Vector4f=<1 1 1 1>
    )
    ScaleRGBA=true
   ) {module=ScaleColor003}
   &ScaleColor(
    Scale Alpha=&FloatFromCurve(
     FloatCurve=$User.FadeinCurve
     CurveIndex=&ClampFloat(Float=&Subtract_Float(A=$System.Age, B=$User.InitDelay))
    )
    ScaleA=true
    ScaleRGB=true
   ) {module=ScaleColor004}
   &ScaleColor(Scale Alpha=3, ScaleA=true, ScaleRGB=true) {module=ScaleColor005}
   /SetVariables(Particles.NewPos:Vector3f=&Lerp_Vector(
     A=&Add_Vector(
      A=Linked() {linkedVariable {name=Particles.Position, type=NiagaraPosition}}
      B=&TransformVector(Vector=<-15 -30 -70>, Destination Space=::World)
     )
     B=Linked() {linkedVariable {name=Particles.Position, type=NiagaraPosition}}
     Alpha=&FloatFromCurve(
      FloatCurve=DataInterface() {propertyValues {Curve {keys [
          {arriveTangent=4.1785, interpMode=RCIM_Cubic, leaveTangent=4.1785, tangentMode=RCTM_User, time=0, value=0}
          {arriveTangent=0.9995, interpMode=RCIM_Cubic, leaveTangent=0.9995, tangentMode=RCTM_User, time=0.2223, value=0.6825}
          {arriveTangent=-0.0405, interpMode=RCIM_Cubic, leaveTangent=-0.0405, tangentMode=RCTM_User, time=1, value=1}
         ]}}}
      CurveIndex=$System.FN_Integrator.Value
     )
    )) {isSetParameters=true}
   &ScaleColor(
    Scale Alpha=&FloatFromCurve(FloatCurve=DataInterface() {propertyValues {Curve {keys [<at 0 1>,<at 0.3 1>,<at 1 1.5>]}}}, CurveIndex=$System.CameraDistanceNormalized)
    ScaleA=true
    ScaleRGB=true
   ) {module=ScaleColor006}
  ]
  rendererClasses [&NiagaraSpriteRendererProperties, &NiagaraRibbonRendererProperties]
  renderers [
   {rendererClass=&NiagaraSpriteRendererProperties}
   {rendererClass=&NiagaraRibbonRendererProperties, rendererIndex=1}
   {rendererClass=&NiagaraRibbonRendererProperties, rendererIndex=2}
   {rendererClass=&NiagaraRibbonRendererProperties, rendererIndex=3}
  ]
 }
 Empty005 {
  emitterUpdateScript [
   &EmitterState(Life Cycle Mode=::NewEnumerator1, Loop Behavior=::NewEnumerator1, Loop Delay=1.1, Loop Duration=1, UseLoopDelay=true)
   &SpawnBurst_Instantaneous(
    Spawn Count=&MakeCustomIntFromBool(Bool=&Boolean_OrOperation(Bool1=$User.bIsFrontEnd, Bool2=$User.bIsFrontEndPreview), True INT=0, False INT=1)
    Use Spawn Probability=false
   )
  ]
  particleSpawnScript [/Niagara/Modules/Spawn/Initialization/V2/InitializeParticle(
    Color=&MakeCustomLinearColorFromBool(
     Bool=&SetBoolByFloatComparison(A=$User.isDark, B=0)
     True Color=<rgba 1 1 1 0.25>
     False Color=<rgba 2 0.6434 0.062 0.25>
    )
    Lifetime=1.15
    Sprite Size=<809.6595 70>
    Sprite Size Mode=::NewEnumerator4
    UsePositionOffset=false
    Write Material Random=true
   )]
  particleUpdateScript [
   &ParticleState()
   &ScaleSpriteSize(Scale Factor=&MakeVector2D(
     X=&FloatFromCurve(FloatCurve=DataInterface() {propertyValues {Curve {keys [
          {arriveTangent=3.4971, interpMode=RCIM_Cubic, leaveTangent=3.4971, tangentMode=RCTM_User, time=0, value=0}
          {arriveTangent=1.4195, interpMode=RCIM_Cubic, leaveTangent=1.4195, tangentMode=RCTM_User, time=0.321, value=0.6968}
          {arriveTangent=-0.0405, interpMode=RCIM_Cubic, leaveTangent=-0.0405, tangentMode=RCTM_User, time=1, value=1}
         ]}}})
     Y=&FloatFromCurve(FloatCurve=DataInterface() {propertyValues {Curve {keys [
          {arriveTangent=-3.0569, interpMode=RCIM_Cubic, leaveTangent=-3.0569, tangentMode=RCTM_User, time=0, value=1}
          {interpMode=RCIM_Cubic, tangentMode=RCTM_User, time=1, value=0}
         ]}}})
    ))
   &ScaleColor(
    Scale Alpha=&FloatFromCurve(FloatCurve=DataInterface() {propertyValues {Curve {keys [
         {interpMode=RCIM_Cubic, tangentMode=RCTM_User, time=0, value=0}
         {arriveTangent=-3.3315, interpMode=RCIM_Cubic, leaveTangent=-3.3315, tangentMode=RCTM_User, time=0.1042, value=0.9849}
         {arriveTangent=-1.0995, interpMode=RCIM_Cubic, leaveTangent=-1.0995, time=0.2166, value=0.1795}
         {tangentMode=RCTM_User, time=1, value=0}
        ]}}})
    ScaleA=true
    ScaleRGB=true
   )
   &SkeletalMeshLocation(
    MaskSpawnGroup=false
    Mesh Sampling Type=::NewEnumerator5
    Skeletal Mesh=$User.SkeletalMesh_DeployBurst
    Socket Sampling Mode=::NewEnumerator5
   )
   &CameraOffset(Camera Offset Amount=50)
  ]
  rendererClasses [&NiagaraSpriteRendererProperties]
  renderers [{rendererClass=&NiagaraSpriteRendererProperties}]
 }
 Empty006 {
  emitterUpdateScript [
   &EmitterState(Life Cycle Mode=::NewEnumerator1, Loop Behavior=::NewEnumerator1, Loop Delay=1.1, Loop Duration=1, UseLoopDelay=true)
   &SpawnBurst_Instantaneous(
    Spawn Count=&MakeCustomIntFromBool(Bool=&Boolean_OrOperation(Bool1=$User.bIsFrontEnd, Bool2=$User.bIsFrontEndPreview), True INT=0, False INT=1)
    Use Spawn Probability=false
   )
  ]
  particleSpawnScript [&InitializeParticle(
    Color=&MakeCustomLinearColorFromBool(
     Bool=&SetBoolByFloatComparison(A=$User.isDark, B=0)
     True Color=<rgba 1 1 1 1>
     False Color=<rgba 1 0.7361 0.594 1>
    )
    Sprite Size Mode=::NewEnumerator3
    Uniform Sprite Size=120
    UsePositionOffset=false
    Write Material Random=true
   )]
  particleUpdateScript [
   &ParticleState()
   &ScaleColor(
    Scale Alpha=&FloatFromCurve(FloatCurve=DataInterface() {propertyValues {Curve {keys [
         {interpMode=RCIM_Cubic, tangentMode=RCTM_User, time=0, value=0}
         {arriveTangent=-3.3315, interpMode=RCIM_Cubic, leaveTangent=-3.3315, tangentMode=RCTM_User, time=0.0948, value=0.9849}
         {arriveTangent=-0.3539, interpMode=RCIM_Cubic, leaveTangent=-0.3539, tangentMode=RCTM_User, time=0.1977, value=0.1165}
         {tangentMode=RCTM_User, time=1, value=0}
        ]}}})
    ScaleA=true
    ScaleRGB=true
   )
   &SkeletalMeshLocation(
    MaskSpawnGroup=false
    Mesh Sampling Type=::NewEnumerator5
    Skeletal Mesh=$User.SkeletalMesh_DeployBurst
    Socket Sampling Mode=::NewEnumerator5
   )
   &CameraOffset(Camera Offset Amount=50)
   &ScaleSpriteSize(Scale Factor=&Vector2DFromFloat(Value=&FloatFromCurve(FloatCurve=DataInterface() {propertyValues {Curve {keys [
          {arriveTangent=4.5784, interpMode=RCIM_Cubic, leaveTangent=4.5784, tangentMode=RCTM_User, time=0, value=0}
          {arriveTangent=0.9995, interpMode=RCIM_Cubic, leaveTangent=0.9995, tangentMode=RCTM_User, time=0.3178, value=0.7257}
          {arriveTangent=-0.0405, interpMode=RCIM_Cubic, leaveTangent=-0.0405, tangentMode=RCTM_User, time=1, value=1}
         ]}}})))
  ]
  rendererClasses [&NiagaraSpriteRendererProperties]
  renderers [
   {rendererClass=&NiagaraSpriteRendererProperties}
   {rendererClass=&NiagaraSpriteRendererProperties, rendererIndex=1}
  ]
 }
 FN_AnimTrail {
  emitterSpawnScript [&FN_RibbonIDOffsetUtility()]
  emitterUpdateScript [
   &EmitterState(Delay First Loop Only=true, Life Cycle Mode=::NewEnumerator1, Loop Delay=$User.InitDelay, UseLoopDelay=true)
   &FN_SpawnAnimTrail(
    Enable:NiagaraBool=&SetBoolByFloatComparison(A=$System.Age, B=$User.InitDelay)
    Spawn Interval=0
    Spawn Mode:ENiagaraSkeletalSpawning=::NewEnumerator0
   )
  ]
  particleSpawnScript [
   &FN_InitializeAnimTrailBoneIndecies(First Bone or Socket Index=1, Second Bone or Socket Index=0)
   /SetVariables(Particles.AlphaFadeIn:NiagaraFloat=/Niagara/DynamicInputs/Math/Min_Float(A:NiagaraFloat=&Divide_Float(A=$Emitter.Age, B=1), B:NiagaraFloat=1)) {isSetParameters=true, module=SetVariables_#1}
   &FN_InitializeAnimTrail(
    AnimTrailWidth=1.8
    Color=&MakeLinearColor(R:NiagaraFloat=1, G:NiagaraFloat=1, B:NiagaraFloat=1, A:NiagaraFloat=1)
    Lifetime=&FloatFromCurve(
     FloatCurve=DataInterface() {propertyValues {Curve {keys [<at 0 0.12>,<at 0.3 0.12>,<at 1 0.36>]}}}
     CurveIndex=$System.CameraDistanceNormalized
    )
    SkeletalMesh=$User.SkeletalMesh003
   )
   &AddVelocity(Velocity=<0 -1000 0>)
  ]
  particleUpdateScript [
   &CurlNoiseForce(
    Noise Frequency=21.1084
    Noise Strength=0
    Pan Noise=true
    Pan Noise Field=<0 0.14 0.15>
    Sample Position:Vector3f=/Niagara/DynamicInputs/Transforms/TransformPosition(Position=Linked() {linkedVariable {name=Particles.Position, type=NiagaraPosition}}, Source Space=::World, Destination Space=::Local)
   )
   &Drag(Drag=3, Use Linear Drag=true, Use Rotational Drag=false)
   &SolveForcesAndVelocity(Clamp Velocity=false, Limit Acceleration=false)
   &FN_UpdateAnimTrail(Locked Particle RibbonWidth=1, SkeletalMesh=$User.SkeletalMesh003) {module=FN_UpdateAnimTrail001}
   /SetVariables(Particles.RibbonU0Override=$Particles.NormalizedAge) {isSetParameters=true, module=SetVariables_#2}
   &ScaleColor(
    Scale Alpha=&FloatFromCurve(
     FloatCurve=DataInterface() {propertyValues {Curve {keys [
         {arriveTangent=-2.6896, interpMode=RCIM_Cubic, leaveTangent=-2.6896, tangentMode=RCTM_User, time=0, value=1}
         {interpMode=RCIM_Cubic, tangentMode=RCTM_User, time=1, value=0}
        ]}}}
     Scale Curve=0.7636
    )
    Scale RGB=&MakeVectorFromLinearColorRGB(Color=&MakeCustomLinearColorFromBool(
      Bool=&SetBoolByFloatComparison(A=$User.isDark, B=0)
      True Color=<rgba 0.01 0.01 0.01 1>
      False Color=<rgba 1 0 0 1>
     ))
    ScaleA=true
    ScaleRGB=true
   )
   &CameraOffset(Camera Offset Amount=30)
   /SetVariables(Particles.RibbonWidth=&FloatFromCurve(FloatCurve=DataInterface() {propertyValues {Curve {keys [<at 0 1.1>,<at 1 1.2>]}}}, Scale Curve=$Particles.Initial.RibbonWidth)) {isSetParameters=true}
   /SetVariables(Particles.Pos:Vector3f=&Add_Vector(A=Linked() {linkedVariable {name=Particles.Position, type=NiagaraPosition}}, B=<0 0 10>)) {isSetParameters=true}
   &ScaleColor(
    Scale Alpha=&FloatFromCurve(
     FloatCurve=DataInterface() {propertyValues {Curve {keys [<at -0.2 -1.5562>,<at 0 0.7>,<at 0.2 0.7>,<at 0.8 0.8>,<at 1 1.2>]}}}
     CurveIndex=$System.Interp.Value
    )
    ScaleA=true
    ScaleRGB=true
   ) {module=ScaleColor001}
   &ScaleColor(Scale Alpha=$System.FN_Integrator.Value, ScaleA=true, ScaleRGB=true) {module=ScaleColor002}
   &ScaleColor(
    Scale Mode=::NewEnumerator1
    Scale RGBA=&MakeCustomVector4FromBool(
     Bool:NiagaraBool=&SetBoolByFloatComparison(A=$User.isDark, B=0)
     True Vector4:Vector4f=<1 1 1 4>
     False Vector4:Vector4f=<1 1 1 1>
    )
    ScaleRGBA=true
   ) {module=ScaleColor003}
   &DynamicMaterialParameters(
    Index 0 Param 1=&MakeCustomFloatFromBool(Bool=&SetBoolByFloatComparison(A=$User.isDark, B=0), True Float=0.5, False Float=1)
    Module.Param0WriteEnabled=true
    Module.Param1WriteEnabled=true
    Module.Param2WriteEnabled=true
    Module.Param3WriteEnabled=true
   )
   &ScaleColor(
    Scale Alpha=&FloatFromCurve(
     FloatCurve=$User.FadeinCurve
     CurveIndex=&ClampFloat(Float=&Subtract_Float(A=$System.Age, B=$User.InitDelay))
    )
    ScaleA=true
    ScaleRGB=true
   ) {module=ScaleColor004}
   &ScaleColor(Scale Alpha=3, ScaleA=true, ScaleRGB=true) {module=ScaleColor005}
   &ScaleColor(
    Scale Alpha=&FloatFromCurve(FloatCurve=DataInterface() {propertyValues {Curve {keys [<at 0 1>,<at 0.3 1>,<at 1 1.5>]}}}, CurveIndex=$System.CameraDistanceNormalized)
    ScaleA=true
    ScaleRGB=true
   ) {module=ScaleColor006}
  ]
  rendererClasses [&NiagaraRibbonRendererProperties, &NiagaraSpriteRendererProperties]
  renderers [
   {rendererClass=&NiagaraRibbonRendererProperties}
   {rendererClass=&NiagaraSpriteRendererProperties, rendererIndex=1}
  ]
 }
 FN_AnimTrail002 {
  emitterSpawnScript [&FN_RibbonIDOffsetUtility()]
  emitterUpdateScript [
   &EmitterState(Delay First Loop Only=true, Life Cycle Mode=::NewEnumerator1, Loop Delay=$User.InitDelay, Loop Duration=1, UseLoopDelay=true)
   &FN_SpawnAnimTrail(
    Enable:NiagaraBool=&SetBoolByFloatComparison(A=$System.Age, B=$User.InitDelay)
    Number of Anim Trails=2
    Spawn Mode:ENiagaraSkeletalSpawning=::NewEnumerator0
   )
  ]
  particleSpawnScript [
   &FN_InitializeAnimTrailBoneIndecies(First Bone or Socket Index=1, Second Bone or Socket Index=0)
   &FN_InitializeAnimTrailBoneIndecies(AnimTrail Selection=1, First Bone or Socket Index=3, Second Bone or Socket Index=2) {module=FN_InitializeAnimTrailBoneIndecies001}
   /SetVariables(Particles.AlphaFadeIn:NiagaraFloat=&Min_Float(A:NiagaraFloat=&Divide_Float(A=$Emitter.Age, B=1), B:NiagaraFloat=1)) {isSetParameters=true, module=SetVariables_#1}
   &FN_InitializeAnimTrail(
    Color=&MakeLinearColor(R:NiagaraFloat=1, G:NiagaraFloat=1, B:NiagaraFloat=1, A:NiagaraFloat=1)
    Lifetime=&FloatFromCurve(FloatCurve=DataInterface() {propertyValues {Curve {keys [<at 0 0.1>,<at 0.3 0.1>,<at 1 0.3>]}}}, CurveIndex=$System.CameraDistanceNormalized)
   )
   &AddVelocity(Velocity=<-100 -500 0>)
  ]
  particleUpdateScript [
   &CurlNoiseForce(
    Noise Frequency=10
    Noise Strength=6118.5103
    Pan Noise=true
    Pan Noise Field=<0 1 1>
    Sample Position:Vector3f=&TransformPosition(Position=Linked() {linkedVariable {name=Particles.Position, type=NiagaraPosition}}, Source Space=::World, Destination Space=::Local)
   )
   &SolveForcesAndVelocity(Clamp Velocity=false, Limit Acceleration=false)
   &FN_UpdateAnimTrail(Locked Particle RibbonWidth=1, SkeletalMesh=$User.SkeletalMesh) {module=FN_UpdateAnimTrail001}
   /SetVariables(Particles.RibbonU0Override=$Particles.NormalizedAge) {isSetParameters=true, module=SetVariables_#2}
   &ScaleColor(
    Scale Alpha=0.2
    Scale RGB=&MakeVectorFromLinearColorRGB(Color=<rgba 1 0 0 1>)
    ScaleA=true
    ScaleRGB=true
   )
   &CameraOffset(Camera Offset Amount=30)
   &ScaleColor(
    Scale Alpha=&FloatFromCurve(
     FloatCurve=DataInterface() {propertyValues {Curve {keys [<at -0.215 0>,<at 0 0.4>,<at 0.8 0.55>,<at 1 0.8>]}}}
     CurveIndex=$System.Interp.Value
     Scale Curve=3
    )
    ScaleA=true
    ScaleRGB=true
   ) {module=ScaleColor001}
   &ScaleColor(Scale Alpha=$System.FN_Integrator.Value, ScaleA=true, ScaleRGB=true) {module=ScaleColor002}
   &ScaleColor(
    Scale Mode=::NewEnumerator1
    Scale RGBA=&MakeCustomVector4FromBool(
     Bool:NiagaraBool=&SetBoolByFloatComparison(A=$User.isDark, B=0)
     True Vector4:Vector4f=<0 0 0 2.5>
     False Vector4:Vector4f=<1 1 1 1>
    )
    ScaleRGBA=true
   ) {module=ScaleColor003}
   &ScaleColor(
    Scale Alpha=&FloatFromCurve(
     FloatCurve=$User.FadeinCurve
     CurveIndex=&ClampFloat(Float=&Subtract_Float(A=$System.Age, B=$User.InitDelay))
    )
    ScaleA=true
    ScaleRGB=true
   ) {module=ScaleColor004}
   &ScaleColor(Scale Alpha=3.5, ScaleA=true, ScaleRGB=true) {module=ScaleColor005}
   &ScaleColor(
    Scale Alpha=&FloatFromCurve(FloatCurve=DataInterface() {propertyValues {Curve {keys [<at 0 1>,<at 0.3 1>,<at 1 1.5>]}}}, CurveIndex=$System.CameraDistanceNormalized)
    ScaleA=true
    ScaleRGB=true
   ) {module=ScaleColor006}
  ]
  rendererClasses [&NiagaraRibbonRendererProperties]
  renderers [{rendererClass=&NiagaraRibbonRendererProperties}]
 }
 FN_AttachedRibbon_v3_HANDS {
  emitterUpdateScript [
   &EmitterState(Delay First Loop Only=true, Life Cycle Mode=::NewEnumerator1, Loop Delay=$User.InitDelay, Loop Duration=1, UseLoopDelay=true)
   &FN_SpawnRibbon_v2(Number of Ribbons=2, Spawn Mode=::NewEnumerator0)
   &FN_Integrator() {enabled=false}
  ]
  particleSpawnScript [&FN_InitializeRibbon_v4(
    Enable Skeletal Position=true
    Lifetime:NiagaraFloat=&FloatFromCurve(
     FloatCurve=DataInterface() {propertyValues {Curve {keys [<at 0 0.2>,<at 0.3 0.2>,<at 1 0.45>]}}}
     CurveIndex=$System.CameraDistanceNormalized
    )
    Offset Axis:ENiagaraOrientationAxis=::ZAxis
    Ribbon Width=48
    SkeletalMesh:NiagaraDataInterfaceSkeletalMesh=$User.SkeletalMesh001
   )]
  particleUpdateScript [
   /SetVariables(Particles.Velocity=&Lerp_Vector(
     A=<0 0 500>
     B=<0 0 -896.2838>
     Alpha=&FloatFromCurve(
      FloatCurve=DataInterface() {propertyValues {Curve {keys [<at 0 0>,<at 0.75 0.1915>,<at 1 1>]}}}
      CurveIndex=&ClampFloat(Float=$System.Interp.Value)
     )
    )) {isSetParameters=true, module=SetVariables_#3}
   &SolveForcesAndVelocity(Clamp Velocity=false, Limit Acceleration=false)
   &FN_UpdateRibbon_v4(Enable Skeletal Position=true, SkeletalMesh:NiagaraDataInterfaceSkeletalMesh=$User.SkeletalMesh001)
   &ScaleColor(
    Scale Alpha=&FloatFromCurve(
     FloatCurve=DataInterface() {propertyValues {Curve {keys [{interpMode=RCIM_Cubic, leaveTangent=4.3733, tangentMode=RCTM_Break, time=0, value=0}, <at 0.5 1>]}}}
     CurveIndex=$Engine.System.Age
    )
    ScaleA=true
    ScaleRGB=true
   )
   &FN_RibbonMaterial_v2(
    Color Pan U Offset:NiagaraFloat=&SelectFloatFromArray(Direct Array Index=$Particles.SpawnGroup, Float Selection Array=DataInterface() {propertyValues {FloatData [0,12.3925,158.6253,64]}})
    Color Pan U Speed=2.5
    Opacity Pan U Offset=0.5
    Opacity Pan U Speed=2
    Write to Module Namespace=true
   )
   &ScaleColor(
    Scale Alpha=&FloatFromCurve(
     FloatCurve=DataInterface() {propertyValues {Curve {keys [
         {arriveTangent=0.0363, interpMode=RCIM_Cubic, leaveTangent=0.0363, tangentMode=RCTM_User, time=0, value=1}
         {interpMode=RCIM_Cubic, tangentMode=RCTM_User, time=0.7, value=0.3}
        ]}}}
     Scale Curve=0.7
    )
    ScaleA=true
    ScaleRGB=true
   ) {module=ScaleColor001}
   /SetVariables(Particles.Lifetime=&SelectFloatFromArray(Direct Array Index=$Particles.SpawnGroup, Float Selection Array=DataInterface() {propertyValues {FloatData [0.15,0.15,0.3,0.3]}})) {enabled=false, isSetParameters=true, module=SetVariables_#4}
   &ScaleColor(
    Scale Alpha=&Add_Float(
     A=&Lerp_Float(A=0.4, B=0.5, Alpha=$System.FN_Integrator.Value)
     B=&Lerp_Float(
      A=0
      B=0.8
      Alpha=&MakeCustomFloatFromBool(
       Bool=&SetBoolByIntComparison(A=$Particles.SpawnGroup, B=0, Comparison Type=::NewEnumerator4)
       True Float=&FloatFromCurve(
        FloatCurve=DataInterface() {propertyValues {Curve {keys [<at 0 0>,<at 1 1>]}}}
        CurveIndex=&ClampFloat(Float=$System.Interp.Value)
       )
       False Float=&FloatFromCurve(
        FloatCurve=DataInterface() {propertyValues {Curve {keys [<at 0 0>,<at 1 1>]}}}
        CurveIndex=&ClampFloat(Float=$System.Interp.Value)
       )
      )
     )
    )
    ScaleA=true
    ScaleRGB=true
   ) {module=ScaleColor002}
   /SetVariables(Particles.Pos:Vector3f=&Add_Vector(
     A=Linked() {linkedVariable {name=Particles.Position, type=NiagaraPosition}}
     B=&MakeVector(
      X=0
      Y=0
      Z=&FloatFromCurve(FloatCurve=DataInterface() {propertyValues {Curve {keys [<at 0 0>,<at 0.95 0>,<at 1 1>]}}}, CurveIndex=$System.Interp.Value, Scale Curve=0)
     )
    )) {isSetParameters=true, module=SetVariables_#5}
   &ScaleColor(
    Scale Alpha=&FloatFromCurve(
     FloatCurve=$User.FadeinCurve
     CurveIndex=&ClampFloat(Float=&Subtract_Float(A=$System.Age, B=$User.InitDelay))
     Scale Curve=1.2
    )
    ScaleA=true
    ScaleRGB=true
   ) {module=ScaleColor003}
   &ScaleColor(
    Scale Alpha=&FloatFromCurve(FloatCurve=DataInterface() {propertyValues {Curve {keys [<at 0 1>,<at 0.3 1>,<at 1 1.1>]}}}, CurveIndex=$System.CameraDistanceNormalized)
    ScaleA=true
    ScaleRGB=true
   ) {module=ScaleColor006}
   &ScaleColor(
    Scale Alpha=&FN_Select_Float_from_Time_Of_Day(Morning=1, Day=1, Evening=1, Night=0.88)
    ScaleA=true
    ScaleRGB=true
   ) {module=ScaleColor004}
  ]
  rendererClasses [&NiagaraRibbonRendererProperties]
  renderers [{rendererClass=&NiagaraRibbonRendererProperties}]
 }
 FN_AttachedRibbon_v3_HANDS001 {
  emitterUpdateScript [
   &EmitterState(Delay First Loop Only=true, Life Cycle Mode=::NewEnumerator1, Loop Delay=$User.InitDelay, Loop Duration=1, UseLoopDelay=true)
   &FN_SpawnRibbon_v2(Number of Ribbons=2, Spawn Mode=::NewEnumerator0)
   &FN_Integrator() {enabled=false}
  ]
  particleSpawnScript [&FN_InitializeRibbon_v4(
    Enable Skeletal Position=true
    Lifetime:NiagaraFloat=&FloatFromCurve(
     FloatCurve=DataInterface() {propertyValues {Curve {keys [<at 0 0.2>,<at 0.3 0.2>,<at 1 0.45>]}}}
     CurveIndex=$System.CameraDistanceNormalized
     Scale Curve=2
    )
    Offset Axis:ENiagaraOrientationAxis=::ZAxis
    Ribbon Width=48
    SkeletalMesh:NiagaraDataInterfaceSkeletalMesh=$User.SkeletalMesh_AirFlow
   )]
  particleUpdateScript [
   /SetVariables(Particles.Velocity=&Lerp_Vector(
     A=<0 0 500>
     B=<0 0 -250>
     Alpha=&FloatFromCurve(
      FloatCurve=DataInterface() {propertyValues {Curve {keys [<at 0 0>,<at 0.75 0.1915>,<at 1 1>]}}}
      CurveIndex=&ClampFloat(Float=$System.Interp.Value)
     )
    )) {isSetParameters=true, module=SetVariables_#3}
   &SolveForcesAndVelocity(Clamp Velocity=false, Limit Acceleration=false)
   &FN_UpdateRibbon_v4(Enable Skeletal Position=true, SkeletalMesh:NiagaraDataInterfaceSkeletalMesh=$User.SkeletalMesh_AirFlow)
   &ScaleColor(
    Scale Alpha=&FloatFromCurve(
     FloatCurve=DataInterface() {propertyValues {Curve {keys [{interpMode=RCIM_Cubic, leaveTangent=4.3733, tangentMode=RCTM_Break, time=0, value=0}, <at 0.5 1>]}}}
     CurveIndex=$Engine.System.Age
    )
    ScaleA=true
    ScaleRGB=true
   )
   &FN_RibbonMaterial_v2(
    Color Pan U Offset:NiagaraFloat=&SelectFloatFromArray(Direct Array Index=$Particles.SpawnGroup, Float Selection Array=DataInterface() {propertyValues {FloatData [0,12.3925,158.6253,64]}})
    Color Pan U Speed=2.5
    Opacity Pan U Offset=0.5
    Opacity Pan U Speed=2
    Write to Module Namespace=true
   )
   &ScaleColor(
    Scale Alpha=&FloatFromCurve(FloatCurve=DataInterface() {propertyValues {Curve {keys [
         {arriveTangent=0.0363, interpMode=RCIM_Cubic, leaveTangent=0.0363, tangentMode=RCTM_User, time=0, value=1}
         {interpMode=RCIM_Cubic, tangentMode=RCTM_User, time=0.7, value=0.3}
        ]}}})
    ScaleA=true
    ScaleRGB=true
   ) {module=ScaleColor001}
   /SetVariables(Particles.Lifetime=&SelectFloatFromArray(Direct Array Index=$Particles.SpawnGroup, Float Selection Array=DataInterface() {propertyValues {FloatData [0.15,0.15,0.3,0.3]}})) {enabled=false, isSetParameters=true, module=SetVariables_#4}
   &ScaleColor(
    Scale Alpha=&Add_Float(
     A=&Lerp_Float(A=0.4, B=0.5, Alpha=$System.FN_Integrator.Value)
     B=&Lerp_Float(
      A=0
      B=0.8
      Alpha=&FloatFromCurve(
       FloatCurve=DataInterface() {propertyValues {Curve {keys [<at 0 0>,<at 1 1>]}}}
       CurveIndex=&ClampFloat(Float=$System.Interp.Value)
      )
     )
    )
    ScaleA=true
    ScaleRGB=true
   ) {module=ScaleColor002}
   /SetVariables(Particles.Pos:Vector3f=&Add_Vector(
     A=Linked() {linkedVariable {name=Particles.Position, type=NiagaraPosition}}
     B=&MakeVector(
      X=0
      Y=0
      Z=&FloatFromCurve(FloatCurve=DataInterface() {propertyValues {Curve {keys [<at 0 0>,<at 0.95 0>,<at 1 1>]}}}, CurveIndex=$System.Interp.Value, Scale Curve=20)
     )
    )) {isSetParameters=true, module=SetVariables_#5}
   /SetVariables(Particles.Pos:Vector3f=&Add_Vector(
     A=$Particles.Pos
     B=&Multiply_VectorByFloat(
      Vector=&TransformVector(Vector=<1 0 0>, Destination Space=::Simulation)
      Float=&SelectFloatFromArray(Direct Array Index=$Particles.SpawnGroup, Float Selection Array=DataInterface() {propertyValues {FloatData [5,-5]}})
     )
    )) {isSetParameters=true}
   &ScaleColor(
    Scale Alpha=&FloatFromCurve(
     FloatCurve=$User.FadeinCurve
     CurveIndex=&ClampFloat(Float=&Subtract_Float(A=$System.Age, B=$User.InitDelay))
     Scale Curve=1.2
    )
    ScaleA=true
    ScaleRGB=true
   ) {module=ScaleColor003}
   &ScaleColor(
    Scale Alpha=&FloatFromCurve(FloatCurve=DataInterface() {propertyValues {Curve {keys [<at 0 1>,<at 0.3 1>,<at 1 1.1>]}}}, CurveIndex=$System.CameraDistanceNormalized)
    ScaleA=true
    ScaleRGB=true
   ) {module=ScaleColor006}
   &ScaleColor(
    Scale Alpha=&FN_Select_Float_from_Time_Of_Day(Morning=1, Day=1, Evening=1, Night=0.88)
    ScaleA=true
    ScaleRGB=true
   ) {module=ScaleColor004}
  ]
  rendererClasses [&NiagaraRibbonRendererProperties]
  renderers [{rendererClass=&NiagaraRibbonRendererProperties}]
 }
 FootWind001 {
  emitterUpdateScript [
   &EmitterState(
    Delay First Loop Only=true
    Inactive Response=::NewEnumerator1
    Life Cycle Mode=::NewEnumerator1
    Loop Delay=$User.InitDelay
    UseLoopDelay=true
   )
   &SpawnBurst_Instantaneous(Spawn Count=1, Use Spawn Probability=false)
  ]
  particleSpawnScript [
   &InitializeParticle(
    Color Mode=::NewEnumerator0
    Lifetime=5
    Mesh Scale=<0.4 0.4 0.4>
    Mesh Scale Mode=::NewEnumerator4
    UsePositionOffset=true
    Write Material Random=true
   )
   &InitialMeshOrientation(Rotation=<-0.13 0 0>, Use Orientation Vector=true, Use Rotation Vector=true)
  ]
  particleUpdateScript [
   &ParticleState()
   &SkeletalMeshLocation(
    Bone / Socket Index=4
    Bone Sampling Mode=::NewEnumerator4
    MaskSpawnGroup=false
    Orientation Sampling=::NewEnumerator1
    Sampled Position Offset=<-80 0 0>
    Skeletal Mesh=$User.SkeletalMesh
   )
   &SkeletalMeshLocation(
    Bone Sampling Mode=::NewEnumerator6
    MaskSpawnGroup=false
    Orientation Sampling=::NewEnumerator1
    Position Sampling=::NewEnumerator2
    Skeletal Mesh=$User.SkeletalMesh
   ) {module=SkeletalMeshLocation001}
   &ScaleColor(Scale Alpha=1.2, ScaleA=true, ScaleRGB=true)
   &ScaleColor(
    Scale Alpha=&Add_Float(
     A=&Lerp_Float(A=1, B=0.55, Alpha=$System.FN_Integrator.Value)
     B=&Lerp_Float(
      A=0
      B=-0.55
      Alpha=&FloatFromCurve(
       FloatCurve=DataInterface() {propertyValues {Curve {keys [<at 0 0>,<at 0.8 0.3>,<at 1 1>]}}}
       CurveIndex=&ClampFloat(Float=$System.Interp.Value)
      )
     )
    )
    ScaleA=true
    ScaleRGB=true
   ) {module=ScaleColor001}
   /SetVariables(Particles.NewO:Quat4f=&MultiplyQuaternion(
     Quaternion A=&MultiplyQuaternion(
      Quaternion A=&FindQuatBetweenTwoVectors(
       From Vector=<0 0 -1>
       To Vector=&Multiply_Vector(
        A=&NormalizeVector(
         Vector To Normalize=&Subtract_Vector(
          A=Linked() {linkedVariable {name=Particles.SkeletalMeshLocation001.SampledPosition, type=NiagaraPosition}}
          B=$Particles.SkeletalMeshLocation.SampledPosition
         )
         Safe Normalize=true
         Safe Fallback Threshold=0
         Safe Fallback Vector=<0 0 0.001>
        )
        B=<-1 1 1>
       )
      )
      Quaternion B=$Particles.MeshOrientation
     )
     Quaternion B=<0 0 0 1>
    )) {isSetParameters=true, module=SetVariables_#6}
   &ScaleColor(
    Scale Alpha=&FloatFromCurve(
     FloatCurve=$User.FadeinCurve
     CurveIndex=&ClampFloat(Float=&Subtract_Float(A=$System.Age, B=$User.InitDelay))
     Scale Curve=1.5
    )
    ScaleA=true
    ScaleRGB=true
   ) {module=ScaleColor003}
  ]
  rendererClasses [&NiagaraMeshRendererProperties]
  renderers [{rendererClass=&NiagaraMeshRendererProperties, rendererIndex=0}]
 }
 Large_Red_FX {
  emitterUpdateScript [
   &EmitterState(
    Delay First Loop Only=true
    Inactive Response=::NewEnumerator1
    Life Cycle Mode=::NewEnumerator1
    Loop Delay=$User.InitDelay
    UseLoopDelay=true
   )
   &SpawnBurst_Instantaneous(Spawn Count=1, Use Spawn Probability=false)
  ]
  particleSpawnScript [
   &InitializeParticle(
    Color=<rgba 0.35 0 0 20>
    Lifetime=5
    Mesh Scale=<4 1 1>
    Mesh Scale Mode=::NewEnumerator4
    Position Mode=::NewEnumerator1
    UsePositionOffset=false
    Write Material Random=true
   )
   &InitialMeshOrientation(Rotation=<0 0.5 0>, Use Orientation Vector=true, Use Rotation Vector=true)
  ]
  particleUpdateScript [
   &ParticleState()
   &SkeletalMeshLocation(Bone / Socket Index=4, Bone Sampling Mode=::NewEnumerator4, MaskSpawnGroup=false, Skeletal Mesh=$User.SkeletalMesh)
   /SetVariables(Particles.MeshOrientation=&FindQuatBetweenTwoVectors(
     From Vector=<1 0 0>
     To Vector=&TransformVector(
      Vector=&NormalizeVector(
       Vector To Normalize=&MakeCustomVectorFromBool(
        Bool=&Boolean_OrOperation(Bool1=$User.bIsFrontEnd, Bool2=$User.bIsFrontEndPreview)
        True Vector=<0 0 -1>
        False Vector=$Engine.Owner.Velocity
       )
       Safe Normalize=true
       Safe Fallback Threshold=0
       Safe Fallback Vector=<0 0 0.001>
      )
      Source Space=::World
      Destination Space=::Local
     )
    )) {isSetParameters=true}
   &ScaleColor(
    Scale Alpha=&Add_Float(
     A=&Lerp_Float(A=0.55, B=0.75, Alpha=$System.FN_Integrator.Value)
     B=&Lerp_Float(
      A=0
      B=0.1
      Alpha=&FloatFromCurve(
       FloatCurve=DataInterface() {propertyValues {Curve {keys [<at 0 0>,<at 0.7 0.3279>,<at 1 1>]}}}
       CurveIndex=&ClampFloat(Float=$System.Interp.Value)
      )
     )
    )
    ScaleA=true
    ScaleRGB=true
   ) {module=ScaleColor001}
   &self:Timer(ScaleDeltaTime:NiagaraFloat=&Lerp_Float(
     A=1
     B=1.5
     Alpha=&FloatFromCurve(
      FloatCurve=DataInterface() {propertyValues {Curve {keys [
          {interpMode=RCIM_Cubic, tangentMode=RCTM_Break, time=0, value=0}
          {interpMode=RCIM_Cubic, tangentMode=RCTM_Break, time=0.8, value=0.1}
          {arriveTangent=0.3068, interpMode=RCIM_Cubic, tangentMode=RCTM_Break, time=1, value=1}
         ]}}}
      CurveIndex=&ClampFloat(Float=$System.Interp.Value)
     )
    )) {module=Timer}
   &DynamicMaterialParameters(
    Index 0 Param 1=$Particles.Timer.Timer
    Module.Param0WriteEnabled=true
    Module.Param1WriteEnabled=true
    Module.Param2WriteEnabled=true
    Module.Param3WriteEnabled=true
   )
   &ScaleColor(
    Scale Alpha=&FloatFromCurve(
     FloatCurve=$User.FadeinCurve
     CurveIndex=&ClampFloat(Float=&Subtract_Float(A=$System.Age, B=$User.InitDelay))
    )
    ScaleA=true
    ScaleRGB=true
   )
   &ScaleColor(
    Scale RGB=&VectorFromFloat(Value=&FloatFromCurve(FloatCurve=DataInterface() {propertyValues {Curve {keys [<at 0 1>,<at 0.3 1>,<at 1 1.7>]}}}, CurveIndex=$System.CameraDistanceNormalized))
    ScaleA=true
    ScaleRGB=true
   ) {module=ScaleColor006}
   &ScaleColor(
    Scale Alpha=&Lerp_Float(A=3, B=1, Alpha=&Lerp_Float(A=1, B=0, Alpha=$User.isDark))
    Scale RGB=&Lerp_Vector(
     A=<0 0 0>
     B=<1 1 1>
     Alpha=&MakeCustomFloatFromBool(Bool=&SetBoolByFloatComparison(A=$User.isDark, B=0), True Float=0, False Float=1)
    )
    ScaleA=true
    ScaleRGB=true
   ) {module=ScaleColor002}
   &ScaleMeshSize(Scale Factor=&VectorFromFloat(Value=&FloatFromCurve(FloatCurve=DataInterface() {propertyValues {Curve {keys [<at 0 1>,<at 0.3 1>,<at 1 1.5>]}}}, CurveIndex=$System.CameraDistanceNormalized)))
  ]
  rendererClasses [&NiagaraMeshRendererProperties]
  renderers [{rendererClass=&NiagaraMeshRendererProperties, rendererIndex=0}]
 }
 Red_Foot_FX {
  emitterUpdateScript [
   &EmitterState(
    Delay First Loop Only=true
    Inactive Response=::NewEnumerator1
    Life Cycle Mode=::NewEnumerator1
    Loop Delay=$User.InitDelay
    UseLoopDelay=true
   )
   &SpawnBurst_Instantaneous(Spawn Count=1, Use Spawn Probability=false)
  ]
  particleSpawnScript [
   &InitializeParticle(
    Color=<rgba 1 0 0 2>
    Lifetime=5
    Mesh Scale=<0.4 0.4 0.4>
    Mesh Scale Mode=::NewEnumerator4
    UsePositionOffset=true
    Write Material Random=true
   )
   &InitialMeshOrientation(Rotation=<-0.13 0 0>, Use Orientation Vector=true, Use Rotation Vector=true)
  ]
  particleUpdateScript [
   &ParticleState()
   &SkeletalMeshLocation(
    Bone / Socket Index=4
    Bone Sampling Mode=::NewEnumerator4
    MaskSpawnGroup=false
    Orientation Sampling=::NewEnumerator1
    Sampled Position Offset=<-80 0 0>
    Skeletal Mesh=$User.SkeletalMesh
   )
   &SkeletalMeshLocation(
    Bone Sampling Mode=::NewEnumerator6
    MaskSpawnGroup=false
    Orientation Sampling=::NewEnumerator1
    Position Sampling=::NewEnumerator2
    Skeletal Mesh=$User.SkeletalMesh
   ) {module=SkeletalMeshLocation001}
   &ScaleColor(Scale Alpha=1.2, ScaleA=true, ScaleRGB=true)
   &ScaleColor(
    Scale Alpha=&Add_Float(
     A=&Lerp_Float(A=1, B=0.55, Alpha=$System.FN_Integrator.Value)
     B=&Lerp_Float(
      A=0
      B=-0.55
      Alpha=&FloatFromCurve(
       FloatCurve=DataInterface() {propertyValues {Curve {keys [<at 0 0>,<at 0.8 0.3>,<at 1 1>]}}}
       CurveIndex=&ClampFloat(Float=$System.Interp.Value)
      )
     )
    )
    ScaleA=true
    ScaleRGB=true
   ) {module=ScaleColor001}
   /SetVariables(Particles.NewO:Quat4f=&MultiplyQuaternion(
     Quaternion A=&MultiplyQuaternion(
      Quaternion A=&FindQuatBetweenTwoVectors(
       From Vector=<0 0 -1>
       To Vector=&Multiply_Vector(
        A=&NormalizeVector(
         Vector To Normalize=&Subtract_Vector(
          A=Linked() {linkedVariable {name=Particles.SkeletalMeshLocation001.SampledPosition, type=NiagaraPosition}}
          B=$Particles.SkeletalMeshLocation.SampledPosition
         )
         Safe Normalize=true
         Safe Fallback Threshold=0
         Safe Fallback Vector=<0 0 0.001>
        )
        B=<-1 1 1>
       )
      )
      Quaternion B=$Particles.MeshOrientation
     )
     Quaternion B=<0 0 0 1>
    )) {isSetParameters=true, module=SetVariables_#6}
   &ScaleColor(
    Scale Alpha=&FloatFromCurve(
     FloatCurve=$User.FadeinCurve
     CurveIndex=&ClampFloat(Float=&Subtract_Float(A=$System.Age, B=$User.InitDelay))
     Scale Curve=1.5
    )
    ScaleA=true
    ScaleRGB=true
   ) {module=ScaleColor003}
  ]
  rendererClasses [&NiagaraMeshRendererProperties]
  renderers [{rendererClass=&NiagaraMeshRendererProperties, rendererIndex=0}]
 }
}
summary {
 systemName=NS_Glider_Emperor
 emitters [
  {emitterName=FN_AnimTrail002, rendererClasses [&NiagaraRibbonRendererProperties]}
  {emitterName=FN_AttachedRibbon_v3_HANDS, rendererClasses [&NiagaraRibbonRendererProperties]}
  {emitterName=AirBreak, rendererClasses [&NiagaraMeshRendererProperties]}
  {emitterName=FootWind001, rendererClasses [&NiagaraMeshRendererProperties]}
  {emitterName=FN_AnimTrail, rendererClasses [&NiagaraRibbonRendererProperties, &NiagaraSpriteRendererProperties]}
  {emitterName=Convergent_airflow_start, rendererClasses [&NiagaraMeshRendererProperties]}
  {emitterName=FN_AttachedRibbon_v3_HANDS001, rendererClasses [&NiagaraRibbonRendererProperties]}
  {emitterName=Empty003, rendererClasses [&NiagaraSpriteRendererProperties, &NiagaraRibbonRendererProperties]}
  {emitterName=Empty005, rendererClasses [&NiagaraSpriteRendererProperties]}
  {emitterName=Empty006, rendererClasses [&NiagaraSpriteRendererProperties]}
  {emitterName=Red_Foot_FX, rendererClasses [&NiagaraMeshRendererProperties]}
  {emitterName=Large_Red_FX, rendererClasses [&NiagaraMeshRendererProperties]}
 ]
 userVariables [
  User.AnimTrail_CapeFlow {
   type=NiagaraDataInterfaceSkeletalMesh
   defaultValue {
    struct=NiagaraExt_VariableValue_DataInterface
    value {dataInterface=&self:NiagaraDataInterfaceSkeletalMesh_4}
   }
  }
  User.bIsFrontEnd {type=NiagaraBool}
  User.bIsFrontEndPreview {type=NiagaraBool}
  User.bisFullyDeployed {type=NiagaraBool, defaultValue=true}
  User.FadeinCurve {
   type=NiagaraDataInterfaceCurve
   description=""
   defaultValue {
    struct=NiagaraExt_VariableValue_DataInterface
    value {
     dataInterface=&self:NiagaraDataInterfaceCurve_0
     dataInterfaceClass=&NiagaraDataInterfaceCurve
    }
   }
  }
  User.ForwardDot {type=NiagaraFloat, defaultValue=0, description=""}
  User.InitDelay {type=NiagaraFloat, defaultValue=1.4, description=""}
  User.isDark {type=NiagaraFloat, defaultValue=0, description=""}
  User.Movement {type=NiagaraFloat, defaultValue=0, description=""}
  User.RibbonIDOffset {type=NiagaraInt32}
  User.RightDot {type=NiagaraFloat, defaultValue=0, description=""}
  User.SkeletalMesh {
   type=NiagaraDataInterfaceSkeletalMesh
   defaultValue {
    struct=NiagaraExt_VariableValue_DataInterface
    value {dataInterface=&self:NiagaraDataInterfaceSkeletalMesh_3}
   }
  }
  User.SkeletalMesh001 {
   type=NiagaraDataInterfaceSkeletalMesh
   defaultValue {
    struct=NiagaraExt_VariableValue_DataInterface
    value {dataInterface=&self:NiagaraDataInterfaceSkeletalMesh_5}
   }
  }
  User.SkeletalMesh002 {
   type=NiagaraDataInterfaceSkeletalMesh
   defaultValue {
    struct=NiagaraExt_VariableValue_DataInterface
    value {dataInterface=&self:NiagaraDataInterfaceSkeletalMesh_6}
   }
  }
  User.SkeletalMesh003 {
   type=NiagaraDataInterfaceSkeletalMesh
   defaultValue {
    struct=NiagaraExt_VariableValue_DataInterface
    value {dataInterface=&self:NiagaraDataInterfaceSkeletalMesh_0}
   }
  }
  User.SkeletalMesh_AirFlow {
   type=NiagaraDataInterfaceSkeletalMesh
   defaultValue {
    struct=NiagaraExt_VariableValue_DataInterface
    value {dataInterface=&self:NiagaraDataInterfaceSkeletalMesh_2}
   }
  }
  User.SkeletalMesh_CapeRed {
   type=NiagaraDataInterfaceSkeletalMesh
   defaultValue {
    struct=NiagaraExt_VariableValue_DataInterface
    value {dataInterface=&self:NiagaraDataInterfaceSkeletalMesh_7}
   }
  }
  User.SkeletalMesh_Deploy {
   type=NiagaraDataInterfaceSkeletalMesh
   defaultValue {
    struct=NiagaraExt_VariableValue_DataInterface
    value {dataInterface=&self:NiagaraDataInterfaceSkeletalMesh_8}
   }
  }
  User.SkeletalMesh_DeployBurst {
   type=NiagaraDataInterfaceSkeletalMesh
   defaultValue {
    struct=NiagaraExt_VariableValue_DataInterface
    value {dataInterface=&self:NiagaraDataInterfaceSkeletalMesh_9}
   }
  }
  User.SkeletalMesh_Flow {
   type=NiagaraDataInterfaceSkeletalMesh
   defaultValue {
    struct=NiagaraExt_VariableValue_DataInterface
    value {dataInterface=&self:NiagaraDataInterfaceSkeletalMesh_1}
   }
  }
 ]
}
